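# Anim Blueprint for the SpartanSmirk face accessory part, which follows the character's pose and adds secondary motion.
# It copies the attached parent's pose, layering Body_Mesh (set on customization load) from both upper arms down.
# That cached pose feeds spine_02-space RigidBody physics, an additive right-hood bone offset, then a Control Rig.
# An arm-yaw alpha is computed but unread here; a second RigidBody node is unconnected.
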AnimBlueprint /BRCosmetics/Characters/Player/Female/Medium/Bodies/F_MED_SpartanSmirk/Meshes/Parts/F_MED_SpartanSmirk_FaceAcc_AnimBP
target_skeleton=&/F_MED_SpartanSmirk_FaceAcc_Skeleton
graphs [
 AnimGraph {nodes [
   RigidBody() {
    bClampLinearTranslationLimitToRefPose=true
    overridePhysicsAsset=&/F_MED_SpartanSmirk_Physics
    simulationSpace=BaseBoneSpace
    alphaBoolBlend {blendOption=Linear}
    baseBoneRef {boneName=spine_02}
    componentAppliedLinearAccClamp=<3000 3000 3000>
    componentLinearAccScale=<0.3 0.4 0.4>
    componentLinearVelScale=<0.3 0.4 0.3>
    simSpaceSettings {maxAngularVelocity=3, maxLinearAcceleration=700, maxLinearVelocity=500, velocityScaleZ=0, worldAlpha=0.65, externalLinearDragV=<0.2 0.2>}
   }
   n1 = RigidBody(ComponentPose=Local To Component(LocalPose=Use cached pose 'Start_Cache'())) {
    bClampLinearTranslationLimitToRefPose=true
    overridePhysicsAsset=&F_MED_SpartanSmirk_Physics
    simulationSpace=BaseBoneSpace
    alphaBoolBlend {blendOption=Linear}
    baseBoneRef {boneName=spine_02}
    componentAppliedLinearAccClamp=<0 0 0>
    simSpaceSettings {maxAngularVelocity=3, maxLinearAcceleration=700, maxLinearVelocity=500, velocityScaleZ=0, worldAlpha=0.65, externalLinearDragV=<0.2 0.2>}
   }
   n2 = "Transform (Modify) Bone - Bone: dyn_R_hood_master"!(ComponentPose=@n1, Alpha=1, Translation="6,0,0") -> (Pose) {translationMode=BMM_Additive, alphaBoolBlend {blendOption=Linear}, boneToModify {boneName=dyn_R_hood_master}, translation {x=6}}
   n3 = Control Rig(Source=Component To Local(ComponentPose=@n2)) {
    alphaBoolBlend {blendOption=Linear}
    controlRigAssetReference {blueprintRigClass=&/F_MED_SpartanSmirk_FaceAcc_CtrlRig.F_MED_SpartanSmirk_FaceAcc_CtrlRig_C}
   }
   Output Pose(Result=Component To Local(ComponentPose=Local To Component(LocalPose=@n3))) {layerGroup=None}
   n4 = Get Body_Mesh()
   Save cached pose 'Start_Cache'(Pose=Layered blend per bone(BasePose=Copy Pose From Mesh() {bCopyCurves=true, bUseAttachedParent=true}, BlendPoses_0=Copy Pose From Mesh(SourceMeshComponent=@n4)) {layerSetup [{branchFilters [{boneName=upperarm_l}, {boneName=upperarm_r}]}]}) {cacheName=Start_Cache}
  ]}
 EventGraph {nodes [
   Event BlueprintUpdateAnimation()
   n1 = BreakTransform(InTransform=GetDeltaTransformFromRefPose(self=Get Body_Mesh(), BoneName=spine_03, BaseName=upperarm_r))
   > Set upperarm_up_alpha(upperarm_up_alpha:double=MapRangeClamped(Value=BreakRotator(InRot=@n1.Rotation).Yaw, InRangeA=70, InRangeB=110)) -> (Output_Get:double)
   n2 = Event OnCharacterCustomizationLoaded()
   > Set Body_Mesh(Body_Mesh=@n2.CustomBaseMesh)
  ]}
]
variables [
 Body_Mesh {type=SkeletalMeshComponent}
 upperarm_up_alpha {type=number}
]
